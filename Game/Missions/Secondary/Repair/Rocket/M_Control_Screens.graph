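# Material for control screens in the rocket repair secondary mission, showing a broken or repaired state on the industrial DamControl prop.
# An icon mask texture on UV channel 2 supplies the screen area in red, a sine-pulsed icon over red tint in green, and a steady icon over green tint in blue.
# "Repaired" blends red to green, "Turned on" scales the glow, and the screen mask also darkens the prop's diffuse texture.
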Material /Game/Missions/Secondary/Repair/Rocket/M_Control_Screens
graph {
 properties {pixelDepthOffsetMode=PDOM_Legacy, previewMesh=&/Prop_DamControl_Screens, usedShadingModels=Default Lit}
 nodes [
  n1 = TextureSample(UVs=TextureCoordinate() {coordinateIndex=2}) {texture=&/T_Screen_Icons_M}
  LinearInterpolate(Alpha=@n1.R) {constA=0.6, constB=0.25}
  n2 = Sine(None=Time())
  n3 = Add(A=Multiply(A=@n1.G, B=ConstantBiasScale(None=@n2)), B=Constant3Vector() {constant=<rgba 0.485 0.0024 0.0024 1>})
  n4 = Add(A=@n1.B, B=Constant3Vector() {constant=<rgba 0.0712 0.475 0.2059 0>})
  n5 = ScalarParameter() {parameterName=Repaired}
  Multiply(
   A=Clamp(None=@n1.R) {maxDefault=0.8}
   B=Multiply(A=ScalarParameter() {parameterName=Turned on}, B=LinearInterpolate(A=@n3, B=@n4, Alpha=@n5))
  )
  Multiply(
   A=TextureSample() {texture=/Game/Environments/Props/Industrial/Textures/Prop_DamControl_02_D}.RGB
   B=LinearInterpolate(Alpha=@n1.R) {constA=1, constB=0.25}
  )
  TextureSample() {samplerType=SAMPLERTYPE_Normal, texture=/Game/Environments/Props/Industrial/Textures/Prop_DamControl_02_N}
 ]
}
